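# Physics asset for the palm tree pickaxe skeletal mesh, giving its fronds and fruit capsule bodies for secondary motion on the melee weapon.
# A long capsule on Melee_TwoHanded_Handle anchors eight leaf chains of three or four bones, each bone constrained to the previous one.
# Leaf joints have very tight swing and twist limits, so any bend spreads along each chain.
# Four fruit bodies attach directly to the handle with slightly looser swing and locked twist.
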PhysicsAsset /BRCosmetics/Weapons/FORT_Melee/Pickaxe_Palm_Tree/Meshes/SK_Pickaxe_Palm_Tree_Physics
bodies {
 Melee_TwoHanded_Handle [{shapeType=Capsule, length=120.0625, radius=2.2442, center=<0.599 -0.9712 32.9515>, rotation=<rot -0.813 -90 0.8382>}]
 dyn_fruit_01 [{shapeType=Capsule, length=1.1178, radius=4.4795, center=<7.3486 0.0677 -0.0218>, rotation=<rot 0 90.2809 -89.0781>}]
 dyn_fruit_02 [{shapeType=Capsule, length=0.7323, radius=6.8883, center=<9.5136 -0.1509 0.0053>, rotation=<rot 0 -89.423 -88.7151>}]
 dyn_fruit_03 [{shapeType=Capsule, length=1.5332, radius=5.9707, center=<9.256 -0.0458 0.0606>, rotation=<rot 0 -91.157 -89.4171>}]
 dyn_fruit_04 [{shapeType=Capsule, length=0.9826, radius=7.8781, center=<10.8394 -0.0513 -0.1034>, rotation=<rot 0 -90.1663 -89.3228>}]
 dyn_leaf_a01 [{shapeType=Capsule, length=6.6474, radius=3.9182, center=<4.9505 0 0.0002>, rotation=<rot 90 0.0395 0.0394>}]
 dyn_leaf_a02 [{shapeType=Capsule, length=9.6267, radius=5.7598, center=<3.5358 0 0.0002>, rotation=<rot 90 -179.9802 -179.9802>}]
 dyn_leaf_a03 [{shapeType=Capsule, length=10.0572, radius=5.3292, center=<4.1336 0 0.0001>, rotation=<rot 90 0.0395 0.0394>}]
 dyn_leaf_a04 [{shapeType=Capsule, length=11.1354, radius=4.2511, center=<3.8203 0 0.0003>, rotation=<rot 90 -0.0198 -0.0199>}]
 dyn_leaf_b01 [{shapeType=Capsule, length=5.683, radius=3.2185, center=<2.1711 0.0003 0>, rotation=<rot -90 0.0099 -0.0099>}]
 dyn_leaf_b02 [{shapeType=Capsule, length=6.4237, radius=4.4316, center=<3.8038 0.0002 0>, rotation=<rot -90 -0.0198 0.0198>}]
 dyn_leaf_b03 [{shapeType=Capsule, length=7.3586, radius=3.4967, center=<3.3108 -0.0001 0>, rotation=<rot -90 -0.0198 0.0197>}]
 dyn_leaf_c01 [{shapeType=Capsule, length=7.1507, radius=3.7046, center=<5.1414 -0.0001 0>, rotation=<rot -90 -0.0197 0.0195>}]
 dyn_leaf_c02 [{shapeType=Capsule, length=6.6482, radius=4.2071, center=<4.4203 0.0001 0>, rotation=<rot -90 -0.0395 0.0394>}]
 dyn_leaf_c03 [{shapeType=Capsule, length=6.3223, radius=2.9422, center=<4.253 0 0>, rotation=<rot -90 -0.0131 0.0129>}]
 dyn_leaf_d01 [{shapeType=Capsule, length=5.4442, radius=3.18, center=<2.14 -0.0001 0>, rotation=<rot -90 0.0396 -0.0396>}]
 dyn_leaf_d02 [{shapeType=Capsule, length=6.9205, radius=3.9348, center=<2.3044 -0.0001 0>, rotation=<rot -90 -0.0132 0.013>}]
 dyn_leaf_d03 [{shapeType=Capsule, length=5.1396, radius=2.9501, center=<1.4879 -0.0002 0>, rotation=<rot -90 0.0197 -0.0195>}]
 dyn_leaf_e01 [{shapeType=Capsule, length=7.2242, radius=4.3496, center=<5.8386 0.0001 0>, rotation=<rot 90 179.9802 179.9803>}]
 dyn_leaf_e02 [{shapeType=Capsule, length=7.4349, radius=4.8716, center=<4.437 -0.0001 0.0002>, rotation=<rot 90 0.0297 0.0298>}]
 dyn_leaf_e03 [{shapeType=Capsule, length=11.1736, radius=4.2129, center=<6.3882 0 0>, rotation=<rot 90 0.0395 0.0394>}]
 dyn_leaf_f01 [{shapeType=Capsule, length=6.3524, radius=3.4067, center=<3.1321 0 -0.0003>, rotation=<rot 90 179.8813 179.8813>}]
 dyn_leaf_f02 [{shapeType=Capsule, length=7.6056, radius=4.0259, center=<3.9638 0 0.0001>, rotation=<rot 90 0.0099 0.0099>}]
 dyn_leaf_f03 [{shapeType=Capsule, length=12.1755, radius=5.8246, center=<7.3101 0 0.0003>, rotation=<rot 90 0.0297 0.0298>}]
 dyn_leaf_f04 [{shapeType=Capsule, length=10.825, radius=4.5615, center=<4.3235 0 -0.0002>, rotation=<rot 90 0.0199 0.02>}]
 dyn_leaf_g01 [{shapeType=Capsule, length=5.9801, radius=3.0341, center=<3.2607 -0.0003 0>, rotation=<rot -90 0.0396 -0.0396>}]
 dyn_leaf_g02 [{shapeType=Capsule, length=7.2255, radius=3.6298, center=<3.5221 0 0>, rotation=<rot -90 -179.8813 179.8813>}]
 dyn_leaf_g03 [{shapeType=Capsule, length=6.5389, radius=4.3165, center=<2.8765 0 0>, rotation=<rot -90 0.0099 -0.0099>}]
 dyn_leaf_g04 [{shapeType=Capsule, length=8.0396, radius=2.8158, center=<2.8493 0 0>, rotation=<rot -90 0.0049 -0.0049>}]
 dyn_leaf_h01 [{shapeType=Capsule, length=7.093, radius=3.4725, center=<3.1119 0 -0.0003>, rotation=<rot 90 0.0099 0.0099>}]
 dyn_leaf_h02 [{shapeType=Capsule, length=10.2718, radius=3.7845, center=<3.7791 0 0.0001>, rotation=<rot 90 0.0198 0.0198>}]
 dyn_leaf_h03 [{shapeType=Capsule, length=9.1265, radius=6.26, center=<4.2503 0 0.0002>, rotation=<rot 90 0.0593 0.0594>}]
 dyn_leaf_h04 [{shapeType=Capsule, length=9.7692, radius=4.0752, center=<4.8119 0 0>, rotation=<rot 90 0.0198 0.0198>}]
}
constraints [
 {bone1Name=dyn_leaf_a01, bone2Name=Melee_TwoHanded_Handle, swing1LimitDegrees=2, swing2LimitDegrees=0.5, twistLimitDegrees=0.1}
 {bone1Name=dyn_leaf_h01, bone2Name=Melee_TwoHanded_Handle, swing1LimitDegrees=2, swing2LimitDegrees=0.5, twistLimitDegrees=0.1}
 {bone1Name=dyn_leaf_f01, bone2Name=Melee_TwoHanded_Handle, swing1LimitDegrees=2, swing2LimitDegrees=0.5, twistLimitDegrees=0.1}
 {bone1Name=dyn_leaf_e01, bone2Name=Melee_TwoHanded_Handle, swing1LimitDegrees=2, swing2LimitDegrees=0.5, twistLimitDegrees=0.1}
 {bone1Name=dyn_leaf_e02, bone2Name=dyn_leaf_e01, swing1LimitDegrees=2, swing2LimitDegrees=0.5, twistLimitDegrees=0.1}
 {bone1Name=dyn_leaf_e03, bone2Name=dyn_leaf_e02, swing1LimitDegrees=2, swing2LimitDegrees=0.5, twistLimitDegrees=0.1}
 {bone1Name=dyn_leaf_f02, bone2Name=dyn_leaf_f01, swing1LimitDegrees=2, swing2LimitDegrees=0.5, twistLimitDegrees=0.1}
 {bone1Name=dyn_leaf_f03, bone2Name=dyn_leaf_f02, swing1LimitDegrees=2, swing2LimitDegrees=0.5, twistLimitDegrees=0.1}
 {bone1Name=dyn_leaf_f04, bone2Name=dyn_leaf_f03, swing1LimitDegrees=2, swing2LimitDegrees=0.5, twistLimitDegrees=0.1}
 {bone1Name=dyn_leaf_h02, bone2Name=dyn_leaf_h01, swing1LimitDegrees=2, swing2LimitDegrees=0.5, twistLimitDegrees=0.1}
 {bone1Name=dyn_leaf_h03, bone2Name=dyn_leaf_h02, swing1LimitDegrees=2, swing2LimitDegrees=0.5, twistLimitDegrees=0.1}
 {bone1Name=dyn_leaf_h04, bone2Name=dyn_leaf_h03, swing1LimitDegrees=2, swing2LimitDegrees=0.5, twistLimitDegrees=0.1}
 {bone1Name=dyn_leaf_a02, bone2Name=dyn_leaf_a01, swing1LimitDegrees=2, swing2LimitDegrees=0.5, twistLimitDegrees=0.1}
 {bone1Name=dyn_leaf_a03, bone2Name=dyn_leaf_a02, swing1LimitDegrees=2, swing2LimitDegrees=0.5, twistLimitDegrees=0.1}
 {bone1Name=dyn_leaf_a04, bone2Name=dyn_leaf_a03, swing1LimitDegrees=2, swing2LimitDegrees=0.5, twistLimitDegrees=0.1}
 {bone1Name=dyn_leaf_b01, bone2Name=Melee_TwoHanded_Handle, swing1LimitDegrees=2, swing2LimitDegrees=0.5, twistLimitDegrees=0.1}
 {bone1Name=dyn_leaf_b02, bone2Name=dyn_leaf_b01, swing1LimitDegrees=2, swing2LimitDegrees=0.5, twistLimitDegrees=0.1}
 {bone1Name=dyn_leaf_b03, bone2Name=dyn_leaf_b02, swing1LimitDegrees=2, swing2LimitDegrees=0.5, twistLimitDegrees=0.1}
 {bone1Name=dyn_leaf_c01, bone2Name=Melee_TwoHanded_Handle, swing1LimitDegrees=2, swing2LimitDegrees=0.5, twistLimitDegrees=0.1}
 {bone1Name=dyn_leaf_c02, bone2Name=dyn_leaf_c01, swing1LimitDegrees=2, swing2LimitDegrees=0.5, twistLimitDegrees=0.1}
 {bone1Name=dyn_leaf_c03, bone2Name=dyn_leaf_c02, swing1LimitDegrees=2, swing2LimitDegrees=0.5, twistLimitDegrees=0.1}
 {bone1Name=dyn_leaf_d01, bone2Name=Melee_TwoHanded_Handle, swing1LimitDegrees=2, swing2LimitDegrees=0.5, twistLimitDegrees=0.1}
 {bone1Name=dyn_leaf_d02, bone2Name=dyn_leaf_d01, swing1LimitDegrees=2, swing2LimitDegrees=0.5, twistLimitDegrees=0.1}
 {bone1Name=dyn_leaf_d03, bone2Name=dyn_leaf_d02, swing1LimitDegrees=2, swing2LimitDegrees=0.5, twistLimitDegrees=0.1}
 {bone1Name=dyn_leaf_g01, bone2Name=Melee_TwoHanded_Handle, swing1LimitDegrees=2, swing2LimitDegrees=0.5, twistLimitDegrees=0.1}
 {bone1Name=dyn_leaf_g02, bone2Name=dyn_leaf_g01, swing1LimitDegrees=2, swing2LimitDegrees=0.5, twistLimitDegrees=0.1}
 {bone1Name=dyn_leaf_g03, bone2Name=dyn_leaf_g02, swing1LimitDegrees=2, swing2LimitDegrees=0.5, twistLimitDegrees=0.1}
 {bone1Name=dyn_leaf_g04, bone2Name=dyn_leaf_g03, swing1LimitDegrees=2, swing2LimitDegrees=0.5, twistLimitDegrees=0.1}
 {
  bone1Name=dyn_fruit_01
  bone2Name=Melee_TwoHanded_Handle
  swing1LimitDegrees=5
  swing2LimitDegrees=5
  twistLimitDegrees=45
  twistMotion=Locked
 }
 {
  bone1Name=dyn_fruit_02
  bone2Name=Melee_TwoHanded_Handle
  swing1LimitDegrees=5
  swing2LimitDegrees=5
  twistLimitDegrees=45
  twistMotion=Locked
 }
 {
  bone1Name=dyn_fruit_03
  bone2Name=Melee_TwoHanded_Handle
  swing1LimitDegrees=5
  swing2LimitDegrees=5
  twistLimitDegrees=45
  twistMotion=Locked
 }
 {
  bone1Name=dyn_fruit_04
  bone2Name=Melee_TwoHanded_Handle
  swing1LimitDegrees=5
  swing2LimitDegrees=5
  twistLimitDegrees=45
  twistMotion=Locked
 }
]
